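# Tinted window material for the PartyBus_RT aircraft, using Default Lit shading with no textures or parameters.
# It is built only from constants, including a near-black grey, and a low-exponent Fresnel that fakes view-dependent glass.
# The Fresnel drives a 1-to-0 lerp, and its inverse scales a warm brown tint in two identical multiplies.
# The graph does not show which material outputs these nodes feed.
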Material /Hera/Athena/Aircraft/PartyBus_RT/Materials/M_PartyBus_RT_TintedWindows
graph {
 properties {pixelDepthOffsetMode=PDOM_Legacy, usedShadingModels=Default Lit}
 nodes [
  Constant3Vector() {constant=<rgba 0.035 0.035 0.035 0>}
  Constant() {r=1}
  Constant()
  Constant()
  n1 = Fresnel() {baseReflectFraction=0.2905, exponent=0.1}
  LinearInterpolate(Alpha=@n1) {constA=1, constB=0}
  n2 = OneMinus(None=@n1)
  n3 = Constant3Vector() {constant=<rgba 0.25 0.1806 0.1224 0>}
  Multiply(A=@n2, B=@n3)
  Multiply(A=@n2, B=@n3)
 ]
}
